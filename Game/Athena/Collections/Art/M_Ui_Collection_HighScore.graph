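# UI material for a collection high-score star icon, tinted by star count, caught state and hover color.
# Three samples of the score texture get their own offset and scale; Is2Stars and Is3stars select how many combine into the mask.
# As wired, second and third star channels subtract from themselves, leaving effectively the first star's blue channel.
# The mask drives color and opacity, with a Use Icon switch for flat output.
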Material /Game/Athena/Collections/Art/M_Ui_Collection_HighScore
graph {
 properties {blendMode=BLEND_Translucent, materialDomain=MD_UI, numCustomizedUVs=1, pixelDepthOffsetMode=PDOM_Legacy, usedShadingModels=Default Lit}
 nodes [
  MaterialFunctionCall(UVs=TextureCoordinate(), Texture Scale=ScalarParameter() {defaultValue=1, parameterName=GlobalScale}) {materialFunction=&ScaleUVsByCenter}
  n1 = VectorParameter() {parameterName=RGB_Hover, defaultValue=<rgba 0 0.0578 0.6038 1>}
  n2 = LinearInterpolate(
   A=VectorParameter() {parameterName=RGB3, defaultValue=<rgba 1 1 1 1>}.RGB
   B=VectorParameter() {parameterName=RGB2, defaultValue=<rgba 1 0.8571 0 1>}.RGB
   Alpha=ScalarParameter() {parameterName=Is3stars, sliderMax=1}
  )
  n3 = Constant() {r=1}
  n4 = TextureSample(
   UVs=MaterialFunctionCall!(
    Texture Scale=ScalarParameter() {defaultValue=1, parameterName=Star1Scale}
    UVs
    Offset XY Texturesize Z=VectorParameter() {parameterName=Star1offset, defaultValue=<rgba 0.25 0.6 1 1>}.RGB
   ) {materialFunction=/Engine/Functions/Engine_MaterialFunctions01/Texturing/ScaleUVsAroundPoint}.UVs
   Tex=TextureObjectParameter() {group=Item, parameterName=ItemIcon, texture=&/T_Ui_Collection_Score_64}
  ) {samplerSource=SSM_Clamp_WorldGroupSettings, texture=/Game/UI/Foundation/Textures/Icons/Athena/T-T-Icon-BR-FlopperGadget}
  n5 = ScalarParameter() {defaultValue=1, parameterName=Star2Scale}
  n6 = ScalarParameter() {parameterName=Is2Stars, sliderMax=1}
  n7 = TextureSample(
   UVs=MaterialFunctionCall(
    Texture Scale=LinearInterpolate(B=@n5, Alpha=@n6) {constA=1}
    Offset XY Texturesize Z=VectorParameter() {parameterName=Star2Offset, defaultValue=<rgba 1 0.8 1 1>}.RGB
   ) {materialFunction=&ScaleUVsAroundPoint}.UVs
   Tex=TextureObjectParameter() {group=Item, parameterName=ItemIcon, texture=&T_Ui_Collection_Score_64}
  ) {samplerSource=SSM_Clamp_WorldGroupSettings, texture=&T-T-Icon-BR-FlopperGadget}
  n8 = Max(A=@n4.B, B=Subtract(A=@n7.G, B=@n7.G))
  n9 = ScalarParameter() {defaultValue=1, parameterName=Star3Scale}
  n10 = ScalarParameter() {parameterName=Is3stars, sliderMax=1}
  n11 = TextureSample(
   UVs=MaterialFunctionCall(
    Texture Scale=LinearInterpolate(B=@n9, Alpha=@n10) {constA=1}
    Offset XY Texturesize Z=VectorParameter() {parameterName=Star3Offset, defaultValue=<rgba 0.5 0.2 1 1>}.RGB
   ) {materialFunction=&ScaleUVsAroundPoint}.UVs
   Tex=TextureObjectParameter() {group=Item, parameterName=ItemIcon, texture=&T_Ui_Collection_Score_64}
  ) {samplerSource=SSM_Clamp_WorldGroupSettings, texture=&T-T-Icon-BR-FlopperGadget}
  n12 = Max(A=@n8, B=Subtract(A=Subtract(A=@n11.R, B=@n11.R), B=@n11.B))
  n13 = If(
   A=ScalarParameter() {parameterName=Is3stars, sliderMax=1}
   B=@n3
   "A > B"=@n12
   "A == B"=@n12
   "A < B"=If(A=ScalarParameter() {parameterName=Is2Stars, sliderMax=1}, B=@n3, "A > B"=@n8, "A == B"=@n8, "A < B"=@n4.B)
  )
  StaticSwitchParameter(
   True=LinearInterpolate(A=@n1.RGB, B=LinearInterpolate(A=@n1.RGB, B=@n2, Alpha=@n13), Alpha=ScalarParameter() {defaultValue=1, parameterName=IsCaught, sliderMax=1})
   False=@n2
  ) {defaultValue=true, parameterName=Use Icon}
  n14 = ScalarParameter() {defaultValue=1, parameterName=Icon_strenght}
  StaticSwitchParameter(True=Max(A=@n13, B=Multiply(A=@n14, B=@n12)), False=Constant() {r=1}) {defaultValue=true, parameterName=Use Icon}
 ]
}
