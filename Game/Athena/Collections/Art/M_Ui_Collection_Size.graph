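# Translucent UI material for a collection size bar, showing a range between MinSize and MaxSize with its upper part, beyond StarTreshold, in a star color.
# Rotated UV masks with adjustable sharpness cut slanted top, left and right borders and place the range edges within them.
# IsDiscovered blends from a flat undiscovered color at reduced opacity to the background, range and star colors at full opacity.
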Material /Game/Athena/Collections/Art/M_Ui_Collection_Size
graph {
 properties {blendMode=BLEND_Translucent, materialDomain=MD_UI, pixelDepthOffsetMode=PDOM_Legacy, usedShadingModels=Default Lit}
 nodes [
  n1 = VectorParameter() {parameterName=Background_RGB, defaultValue=<rgba 0 0.0409 0.7835 1>}
  n2 = VectorParameter() {parameterName=RANGE_RGB, defaultValue=<rgba 0 0.9563 1 1>}
  n3 = ComponentMask(None=Rotator(
    Coordinate=TextureCoordinate() {coordinateIndex=4}
    Time=ScalarParameter() {defaultValue=-0.0025, group=Border Angles, parameterName=Vert-Angle}
   ) {speed=6.2832}) {r=true}
  n4 = Subtract(
   A=@n3
   B=LinearInterpolate(
    A=ScalarParameter() {defaultValue=0.009, group=Border_Range, parameterName=Left-RangeOffset, sliderMax=1}
    B=ScalarParameter() {defaultValue=0.985, group=Border_Range, parameterName=Right-RangeOffset, sliderMax=1}
    Alpha=ScalarParameter() {defaultValue=0.75, parameterName=MaxSize, sliderMax=1}
   )
  )
  n5 = ScalarParameter() {defaultValue=128, group=Border sharpness, parameterName=VerticalSharpness}
  n6 = OneMinus(None=Saturate(None=Multiply(A=@n4, B=@n5)))
  n7 = Subtract(
   A=@n3
   B=LinearInterpolate(
    A=ScalarParameter() {defaultValue=0.009, group=Border_Range, parameterName=Left-RangeOffset, sliderMax=1}
    B=ScalarParameter() {defaultValue=0.985, group=Border_Range, parameterName=Right-RangeOffset, sliderMax=1}
    Alpha=ScalarParameter() {defaultValue=0.25, parameterName=MinSize, sliderMax=1}
   )
  )
  n8 = DotProduct(A=@n6, B=Saturate(None=Multiply(A=@n7, B=@n5)))
  n9 = ScalarParameter() {defaultValue=0.25, parameterName=MinSize, sliderMax=1}
  n10 = Subtract(A=ScalarParameter() {defaultValue=0.75, parameterName=MaxSize, sliderMax=1}, B=@n9)
  n11 = ScalarParameter() {defaultValue=0.8, parameterName=StarTreshold, sliderMax=1}
  n12 = LinearInterpolate(
   A=ScalarParameter() {defaultValue=0.009, group=Border_Range, parameterName=Left-RangeOffset, sliderMax=1}
   B=ScalarParameter() {defaultValue=0.985, group=Border_Range, parameterName=Right-RangeOffset, sliderMax=1}
   Alpha=Add(A=Multiply(A=@n10, B=@n11) {constB=0.8}, B=@n9)
  )
  n13 = Saturate(None=Multiply(A=Subtract(A=@n3, B=@n12), B=@n5))
  LinearInterpolate(
   A=VectorParameter() {parameterName=Undiscovered_RGB, defaultValue=<rgba 0 0.004 0.0723 1>}.RGB
   B=LinearInterpolate(
    A=LinearInterpolate(A=@n1.RGB, B=@n2.RGB, Alpha=@n8)
    B=VectorParameter() {parameterName=StarColor, defaultValue=<rgba 1 0.8571 0 1>}.RGB
    Alpha=DotProduct(A=@n6, B=@n13)
   )
   Alpha=ScalarParameter() {parameterName=IsDiscovered, sliderMax=1}
  )
  n14 = ComponentMask(None=Rotator(Coordinate=TextureCoordinate() {coordinateIndex=4}, Time=ScalarParameter() {defaultValue=0.05, group=Border Angles, parameterName=Top-Angle}) {speed=6.2832}) {g=true}
  n15 = ScalarParameter() {defaultValue=0.1857, group=Border_Range, parameterName=Top-RangeOffset, sliderMax=1}
  n16 = Saturate(None=Multiply(A=Subtract(A=@n14, B=@n15), B=ScalarParameter() {defaultValue=8, group=Border sharpness, parameterName=HorizontalSharpness}))
  n17 = ScalarParameter() {defaultValue=0.009, group=Border_Range, parameterName=Left-RangeOffset, sliderMax=1}
  n18 = Saturate(None=Multiply(A=Subtract(A=@n3, B=@n17), B=@n5))
  n19 = ScalarParameter() {defaultValue=0.985, group=Border_Range, parameterName=Right-RangeOffset, sliderMax=1}
  n20 = Saturate(None=Multiply(A=Subtract(A=@n3, B=@n19), B=@n5))
  Subtract(
   A=DotProduct(A=DotProduct(A=@n16, B=@n18), B=OneMinus(None=@n20))
   B=LinearInterpolate(
    A=ScalarParameter() {defaultValue=0.5, parameterName=Undiscoverd_opacity, sliderMax=1}
    B=Constant()
    Alpha=ScalarParameter() {parameterName=IsDiscovered, sliderMax=1}
   )
  )
 ]
}
