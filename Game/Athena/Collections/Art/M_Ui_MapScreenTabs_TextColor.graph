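# UI material for map screen tab text color, blending between hover and selection states.
# A Hover scalar interpolates between two blue color parameters, RGB1 and RGB2.
# A Select scalar then blends that result toward the near-black RGB3, so full selection overrides the hover color.
# The material is translucent and uses no textures, only parameters and a constant.
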Material /Game/Athena/Collections/Art/M_Ui_MapScreenTabs_TextColor
graph {
 properties {blendMode=BLEND_Translucent, materialDomain=MD_UI, pixelDepthOffsetMode=PDOM_Legacy, usedShadingModels=Default Lit}
 nodes [
  Constant() {r=1}
  LinearInterpolate(
   A=LinearInterpolate(
    A=VectorParameter() {parameterName=RGB1, defaultValue=<rgba 0 0.3763 0.8796 1>}.RGB
    B=VectorParameter() {parameterName=RGB2, defaultValue=<rgba 0.2159 0.6584 1 1>}.RGB
    Alpha=ScalarParameter() {parameterName=Hover, sliderMax=1}
   )
   B=VectorParameter() {parameterName=RGB3, defaultValue=<rgba 0.0024 0.0027 0.0144 1>}.RGB
   Alpha=ScalarParameter() {parameterName=Select, sliderMax=1}
  )
 ]
}
