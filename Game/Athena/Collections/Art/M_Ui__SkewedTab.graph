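# M_Ui__SkewedTab is a translucent UI material for a skewed tab in the Athena Collections art.
# It offsets and rotates normalized UI UVs using bottom position and rotation parameters, each blended toward a selected variant by Select.
# The rotated coordinate's green channel, scaled by sharpness and saturated, forms an angled edge blending white to yellow.
# A separate branch builds an inverted horizontal falloff mask from centered, absolute UVs.
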Material /Game/Athena/Collections/Art/M_Ui__SkewedTab
graph {
 properties {blendMode=BLEND_Translucent, materialDomain=MD_UI, pixelDepthOffsetMode=PDOM_Legacy, usedShadingModels=Default Lit}
 nodes [
  n1 = MaterialFunctionCall() {materialFunction=&GetUserInterfaceUV}
  n2 = LinearInterpolate(
   A=ScalarParameter() {defaultValue=0.5, parameterName=bottom_Position}
   B=ScalarParameter() {defaultValue=0.5, parameterName=bottom_Position_Select}
   Alpha=ScalarParameter() {parameterName=Select, sliderMax=1}
  )
  n3 = ScalarParameter() {defaultValue=0.5, parameterName=bottom_Rotation}
  n4 = ScalarParameter() {defaultValue=0.5, parameterName=bottom_Rotation_Select}
  n5 = ScalarParameter() {parameterName=Select, sliderMax=1}
  n6 = ComponentMask(None=Rotator(Coordinate=Subtract(A=@n1.Normalized UV, B=@n2), Time=LinearInterpolate(A=@n3, B=@n4, Alpha=@n5))) {g=true}
  n7 = ScalarParameter() {defaultValue=128, parameterName=bottom_Sharpness}
  LinearInterpolate(
   A=Constant4Vector() {constant=<rgba 1 1 1 1>}.RGBA
   B=Constant4Vector() {constant=<rgba 0.9301 1 0.0103 1>}.RGBA
   Alpha=Saturate(None=Multiply(A=@n6, B=@n7))
  ) {constAlpha=0}
  n8 = Add(A=@n1.Normalized UV) {constB=-0.5}
  n9 = Subtract(A=ComponentMask(None=Abs(None=@n8)) {r=true}, B=ScalarParameter() {defaultValue=0.45, group=Item, parameterName=falloff_position})
  n10 = ScalarParameter() {defaultValue=10, group=Item, parameterName=falloff_Strenght}
  OneMinus(None=Saturate(None=Multiply(A=@n9, B=@n10) {constB=32}))
 ]
}
