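# UI material in the Athena Collections art folder, named for changing text color on hover.
# It linearly interpolates between a MainColor and a HoverColor vector parameter, both dark blues by default, using the HoverControl scalar parameter as the blend alpha.
# The material uses the UI domain with translucent blending and has no textures or time-based animation, so the blend depends on the HoverControl value.
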Material /Game/Athena/Collections/Art/M_Ui_textColorHover
graph {
 properties {blendMode=BLEND_Translucent, materialDomain=MD_UI, pixelDepthOffsetMode=PDOM_Legacy, usedShadingModels=Default Lit}
 nodes [LinearInterpolate(
   A=VectorParameter() {parameterName=MainColor, defaultValue=<rgba 0 0.0005 0.0104 1>}.RGB
   B=VectorParameter() {parameterName=HoverColor, defaultValue=<rgba 0 0.004 0.0723 1>}.RGB
   Alpha=ScalarParameter() {parameterName=HoverControl}
  )]
}
